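# Sound cue for ambient spot sounds of water impacts and splashes on a metal pontoon, looping at irregular intervals.
# A Looping node repeats a random 3 to 8 second delay, then a slight pitch modulation over an equal-weight random pick from fourteen splash waves.
# It reuses a wood-creak attenuation asset and the Fort_Doors sound class, plays at reduced volume with a tight concurrency override, and sends to the default reverb submix.
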SoundCue /Game/Sounds/Helios/Environment_Actors/MetalPontoon/Spot_Water_Metal_Pontoon_Impact_Splash_01_Cue
root=Looping_0
nodes [
 {id=Delay_0, children [Modulator_3], properties {delayMax=8, delayMin=3}}
 {id=Looping_0, children [Delay_0]}
 {id=Modulator_3, children [Random_0], properties {pitchMax=1.1, pitchMin=0.9}}
 {
  id=Random_0
  children [
   WavePlayer_0
   WavePlayer_1
   WavePlayer_2
   WavePlayer_3
   WavePlayer_4
   WavePlayer_5
   WavePlayer_6
   WavePlayer_7
   WavePlayer_8
   WavePlayer_9
   WavePlayer_10
   WavePlayer_11
   WavePlayer_12
   WavePlayer_13
  ]
  properties {weights [1,1,1,1,1,1,1,1,1,1,1,1,1,1]}
 }
 {
  id=WavePlayer_0
  properties {soundWaveAssetPtr=&/Spot_Water_Metal_Pontoon_Impact_Splash_01}
 }
 {
  id=WavePlayer_1
  properties {soundWaveAssetPtr=&/Spot_Water_Metal_Pontoon_Impact_Splash_02}
 }
 {
  id=WavePlayer_10
  properties {soundWaveAssetPtr=&/Spot_Water_Metal_Pontoon_Impact_Splash_11}
 }
 {
  id=WavePlayer_11
  properties {soundWaveAssetPtr=&/Spot_Water_Metal_Pontoon_Impact_Splash_12}
 }
 {
  id=WavePlayer_12
  properties {soundWaveAssetPtr=&/Spot_Water_Metal_Pontoon_Impact_Splash_13}
 }
 {
  id=WavePlayer_13
  properties {soundWaveAssetPtr=&/Spot_Water_Metal_Pontoon_Impact_Splash_14}
 }
 {
  id=WavePlayer_2
  properties {soundWaveAssetPtr=&/Spot_Water_Metal_Pontoon_Impact_Splash_03}
 }
 {
  id=WavePlayer_3
  properties {soundWaveAssetPtr=&/Spot_Water_Metal_Pontoon_Impact_Splash_04}
 }
 {
  id=WavePlayer_4
  properties {soundWaveAssetPtr=&/Spot_Water_Metal_Pontoon_Impact_Splash_05}
 }
 {
  id=WavePlayer_5
  properties {soundWaveAssetPtr=&/Spot_Water_Metal_Pontoon_Impact_Splash_06}
 }
 {
  id=WavePlayer_6
  properties {soundWaveAssetPtr=&/Spot_Water_Metal_Pontoon_Impact_Splash_07}
 }
 {
  id=WavePlayer_7
  properties {soundWaveAssetPtr=&/Spot_Water_Metal_Pontoon_Impact_Splash_08}
 }
 {
  id=WavePlayer_8
  properties {soundWaveAssetPtr=&/Spot_Water_Metal_Pontoon_Impact_Splash_09}
 }
 {
  id=WavePlayer_9
  properties {soundWaveAssetPtr=&/Spot_Water_Metal_Pontoon_Impact_Splash_10}
 }
]
properties {
 attenuationSettings=/Game/Sounds/Artemis/POI/Chrome_BugleBay/Spot/Fort_Oneshot_Wood_Creak_Int_Floor_Cone_Att
 bOverrideConcurrency=true
 maxDistance=1074
 soundClassObject=&Fort_Doors
 volumeMultiplier=0.4
 concurrencyOverrides {maxCount=2, voiceStealReleaseTime=2}
 soundSubmixSends [{minSendLevel=0.6, soundSubmix=&Fortnite_Reverb_Default}]
}
