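# Sound cue that plays recurring metal pontoon water impact splashes as an environmental spot sound.
# A looping node repeats a random delay of a few seconds, slight pitch modulation and an equally weighted random pick from fourteen splash waves.
# It borrows a wood-creak attenuation asset and the Fort_Doors sound class, plays at reduced volume with overridden low-count concurrency, and sends to the default reverb submix.
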SoundCue /Game/Sounds/Helios/Environment_Actors/MetalPontoon/Spot_Water_Metal_Pontoon_Impact_Splash_02_Cue
root=Looping_0
nodes [
 {id=Delay_0, children [Modulator_3], properties {delayMax=8, delayMin=3}}
 {id=Looping_0, children [Delay_0]}
 {id=Modulator_3, children [Random_3], properties {pitchMax=1.1, pitchMin=0.9}}
 {
  id=Random_3
  children [
   WavePlayer_29
   WavePlayer_30
   WavePlayer_31
   WavePlayer_32
   WavePlayer_33
   WavePlayer_34
   WavePlayer_35
   WavePlayer_36
   WavePlayer_37
   WavePlayer_38
   WavePlayer_39
   WavePlayer_40
   WavePlayer_41
   WavePlayer_42
  ]
  properties {weights [1,1,1,1,1,1,1,1,1,1,1,1,1,1]}
 }
 {
  id=WavePlayer_29
  properties {soundWaveAssetPtr=&/Spot_Water_Metal_Pontoon_Impact_Splash_15}
 }
 {
  id=WavePlayer_30
  properties {soundWaveAssetPtr=&/Spot_Water_Metal_Pontoon_Impact_Splash_16}
 }
 {
  id=WavePlayer_31
  properties {soundWaveAssetPtr=&/Spot_Water_Metal_Pontoon_Impact_Splash_17}
 }
 {
  id=WavePlayer_32
  properties {soundWaveAssetPtr=&/Spot_Water_Metal_Pontoon_Impact_Splash_18}
 }
 {
  id=WavePlayer_33
  properties {soundWaveAssetPtr=&/Spot_Water_Metal_Pontoon_Impact_Splash_19}
 }
 {
  id=WavePlayer_34
  properties {soundWaveAssetPtr=&/Spot_Water_Metal_Pontoon_Impact_Splash_20}
 }
 {
  id=WavePlayer_35
  properties {soundWaveAssetPtr=&/Spot_Water_Metal_Pontoon_Impact_Splash_21}
 }
 {
  id=WavePlayer_36
  properties {soundWaveAssetPtr=&/Spot_Water_Metal_Pontoon_Impact_Splash_22}
 }
 {
  id=WavePlayer_37
  properties {soundWaveAssetPtr=&/Spot_Water_Metal_Pontoon_Impact_Splash_23}
 }
 {
  id=WavePlayer_38
  properties {soundWaveAssetPtr=&/Spot_Water_Metal_Pontoon_Impact_Splash_24}
 }
 {
  id=WavePlayer_39
  properties {soundWaveAssetPtr=&/Spot_Water_Metal_Pontoon_Impact_Splash_25}
 }
 {
  id=WavePlayer_40
  properties {soundWaveAssetPtr=&/Spot_Water_Metal_Pontoon_Impact_Splash_26}
 }
 {
  id=WavePlayer_41
  properties {soundWaveAssetPtr=&/Spot_Water_Metal_Pontoon_Impact_Splash_27}
 }
 {
  id=WavePlayer_42
  properties {soundWaveAssetPtr=&/Spot_Water_Metal_Pontoon_Impact_Splash_28}
 }
]
properties {
 attenuationSettings=/Game/Sounds/Artemis/POI/Chrome_BugleBay/Spot/Fort_Oneshot_Wood_Creak_Int_Floor_Cone_Att
 bOverrideConcurrency=true
 maxDistance=1074
 soundClassObject=&Fort_Doors
 volumeMultiplier=0.4
 attenuationOverrides {falloffDistance=1000, attenuationShapeExtents {x=50}}
 concurrencyOverrides {maxCount=2, voiceStealReleaseTime=2}
 soundSubmixSends [{minSendLevel=0.6, soundSubmix=&Fortnite_Reverb_Default}]
}
